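# M_Sequins_Example is a Default Lit example material for the cosmetic sequins material function, MFW_Sequins.
# It uses Material Attributes, and its graph holds a single call to that function, which exposes IN_MA, UVs, TechArtMask and Fresnel pins.
# No other nodes or parameters are present, so any sequin logic lives inside the function.
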Material /Game/Global/MaterialFunctions/Cosmetics/Sequins/Examples/M_Sequins_Example
graph {
 properties {bAutomaticallySetUsageInEditor=false, bUseMaterialAttributes=true, usedShadingModels=Default Lit}
 nodes [MaterialFunctionCall!(IN_MA, UVs, TechArtMask, Fresnel) {materialFunction=/Game/Global/MaterialFunctions/Cosmetics/Sequins/MFW_Sequins}]
}
